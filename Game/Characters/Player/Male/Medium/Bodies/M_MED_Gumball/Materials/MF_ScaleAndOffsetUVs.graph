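# Material function in the Gumball player body's materials folder that scales and offsets UVs and outputs a matching 0-1 mask.
# It divides UVs by Texture Scale, corrects the center so scaling pivots on 0.5, then adds Offset.
# The mask multiplies two one-dimensional SphereMasks on the transformed U and V around 0.5, bounding the placed texture area.
# U radius is fixed, while V radius is adjustable through Y_MaskRadius.
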MaterialFunction /Game/Characters/Player/Male/Medium/Bodies/M_MED_Gumball/Materials/MF_ScaleAndOffsetUVs
graph {nodes [
  n1 = FunctionInput(Preview=TextureCoordinate()) {
   bUsePreviewValueAsDefault=true
   description=UVs. Only need to plug in if you want a different UV channel.
   inputName=UVs
   inputType=FunctionInput_Vector2
  }
  n2 = FunctionInput(Preview=Constant() {r=1}) {bUsePreviewValueAsDefault=true, description="Scale factor. Default = 1", inputName=Texture Scale, inputType=FunctionInput_Vector2}
  n3 = Divide(A=@n1, B=@n2)
  n4 = Constant() {r=0.5}
  n5 = Add(
   A=Subtract(A=Add(A=@n3, B=@n4), B=Divide(A=@n4, B=@n2) {desc=Scale the center offset by the scaling factor}) {desc=Re offset to maintain center}
   B=FunctionInput(Preview=Constant2Vector()) {bUsePreviewValueAsDefault=true, inputName=Offset, inputType=FunctionInput_Vector2}
  )
  FunctionOutput(None=@n5) {outputName=UVs}
  n6 = Constant() {r=0.5}
  n7 = MaterialFunctionCall(Float2=@n5) {materialFunction=&BreakOutFloat2Components}
  n8 = FunctionInput(Preview=Constant() {r=0.65}) {bUsePreviewValueAsDefault=true, inputName=Y_MaskRadius, inputType=FunctionInput_Scalar}
  FunctionOutput(None=Multiply(A=SphereMask(A=@n6, B=@n7.R) {attenuationRadius=0.5}, B=SphereMask(A=@n7.G, B=@n6, Radius=@n8) {attenuationRadius=0.65})) {outputName=0-1 mask, sortPriority=1}
 ]}
